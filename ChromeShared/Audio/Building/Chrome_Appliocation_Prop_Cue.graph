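# Sound cue for the Chrome prop application sound, played when chrome is applied to a building prop, in the gadgets sound class.
# A quality-level switch sends its first three slots to one equal-weight random pick of six waves and leaves the fourth empty.
# A modulator varies pitch slightly.
# World attenuation applies, virtualization is off, and a concurrency override tightly caps simultaneous instances with volume scaling.
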SoundCue /ChromeShared/Audio/Building/Chrome_Appliocation_Prop_Cue
root=Modulator_0
nodes [
 {id=Modulator_0, children [QualityLevel_0], properties {pitchMax=1.1, pitchMin=0.9, volumeMax=1, volumeMin=1}}
 {id=QualityLevel_0, children [Random_1,Random_1,Random_1,null]}
 {
  id=Random_1
  children [WavePlayer_6, WavePlayer_7, WavePlayer_8, WavePlayer_9, WavePlayer_10, WavePlayer_11]
  properties {bShouldExcludeFromBranchCulling=true, weights [1,1,1,1,1,1]}
 }
 {id=WavePlayer_10, properties {soundWaveAssetPtr=&/Chrome_Build_Prop_Application_05}}
 {id=WavePlayer_11, properties {soundWaveAssetPtr=&/Chrome_Build_Prop_Application_06}}
 {id=WavePlayer_6, properties {soundWaveAssetPtr=&/Chrome_Build_Prop_Application_01}}
 {id=WavePlayer_7, properties {soundWaveAssetPtr=&/Chrome_Build_Prop_Application_02}}
 {id=WavePlayer_8, properties {soundWaveAssetPtr=&/Chrome_Build_Prop_Application_03}}
 {id=WavePlayer_9, properties {soundWaveAssetPtr=&/Chrome_Build_Prop_Application_04}}
]
properties {
 attenuationSettings=&/Chrome_Application_World_Att
 bOverrideConcurrency=true
 maxDistance=6000
 soundClassObject=&Fort_Gadgets
 virtualizationMode=Disabled
 volumeMultiplier=0.4902
 concurrencyOverrides {bVolumeScaleCanRelease=true, maxCount=2, volumeScale=0.5, volumeScaleAttackTime=0.1}
}
