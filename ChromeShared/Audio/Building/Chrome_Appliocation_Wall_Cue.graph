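# Sound cue for the Chrome wall application effect, played as a world-attenuated sound in the gadgets sound class.
# A random node picks one of six equally weighted wall application waves, and a root modulator adds slight pitch variation.
# A quality level node shares that random set across its first three slots and leaves the fourth empty.
# Concurrency override allows two instances with volume scaling; virtualization is disabled.
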SoundCue /ChromeShared/Audio/Building/Chrome_Appliocation_Wall_Cue
root=Modulator_0
nodes [
 {id=Modulator_0, children [QualityLevel_0], properties {pitchMax=1.1, pitchMin=0.9, volumeMax=1, volumeMin=1}}
 {id=QualityLevel_0, children [Random_0,Random_0,Random_0,null]}
 {
  id=Random_0
  children [WavePlayer_0,WavePlayer_1,WavePlayer_2,WavePlayer_3,WavePlayer_4,WavePlayer_5]
  properties {bShouldExcludeFromBranchCulling=true, weights [1,1,1,1,1,1]}
 }
 {id=WavePlayer_0, properties {soundWaveAssetPtr=&/Chrome_Build_Wall_Application_01}}
 {id=WavePlayer_1, properties {soundWaveAssetPtr=&/Chrome_Build_Wall_Application_02}}
 {id=WavePlayer_2, properties {soundWaveAssetPtr=&/Chrome_Build_Wall_Application_03}}
 {id=WavePlayer_3, properties {soundWaveAssetPtr=&/Chrome_Build_Wall_Application_04}}
 {id=WavePlayer_4, properties {soundWaveAssetPtr=&/Chrome_Build_Wall_Application_05}}
 {id=WavePlayer_5, properties {soundWaveAssetPtr=&/Chrome_Build_Wall_Application_06}}
]
properties {
 attenuationSettings=&/Chrome_Application_World_Att
 bOverrideConcurrency=true
 maxDistance=6000
 soundClassObject=&Fort_Gadgets
 virtualizationMode=Disabled
 volumeMultiplier=0.4902
 concurrencyOverrides {bVolumeScaleCanRelease=true, maxCount=2, volumeScale=0.5, volumeScaleAttackTime=0.1}
}
