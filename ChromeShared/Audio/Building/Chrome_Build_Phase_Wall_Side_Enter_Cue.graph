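# Sound cue for entering a Chrome building wall from the side, playing one of four metal phase sounds at random with raised pitch variation.
# A PlayerAffliation switch sends its first two inputs to a boosted first-person foley branch with focus attenuation.
# The other two use enemy third-person foley class and standard wall attenuation, one much quieter.
# Virtualization is off, and a concurrency override caps simultaneous instances, stopping lowest priority.
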SoundCue /ChromeShared/Audio/Building/Chrome_Build_Phase_Wall_Side_Enter_Cue
root=Switch_0
nodes [
 {id=Attenuation_0, children [SoundClass_1], properties {attenuationSettings=&/Chrome_Build_Phase_Wall_Focus_Att}}
 {id=Attenuation_1, children [SoundClass_0], properties {attenuationSettings=&/Chrome_Build_Phase_Wall_Att}}
 {id=Attenuation_2, children [SoundClass_2], properties {attenuationSettings=&Chrome_Build_Phase_Wall_Att}}
 {id=Mixer_0, children [Modulator_0], properties {inputVolume [4.35]}}
 {id=Mixer_1, children [Modulator_5], properties {inputVolume [0.1]}}
 {id=Mixer_2, children [Modulator_6], properties {inputVolume [1]}}
 {id=Modulator_0, children [Random_0], properties {pitchMax=1.5, pitchMin=1}}
 {id=Modulator_5, children [Random_5], properties {pitchMax=1.5, pitchMin=1}}
 {id=Modulator_6, children [Random_6], properties {pitchMax=1.5, pitchMin=1}}
 {
  id=Random_0
  children [WavePlayer_0,WavePlayer_1,WavePlayer_2,WavePlayer_3]
  properties {bShouldExcludeFromBranchCulling=true, weights [1,1,1,1]}
 }
 {
  id=Random_5
  children [WavePlayer_8, WavePlayer_13, WavePlayer_14, WavePlayer_15]
  properties {bShouldExcludeFromBranchCulling=true, weights [1,1,1,1]}
 }
 {
  id=Random_6
  children [WavePlayer_16, WavePlayer_17, WavePlayer_18, WavePlayer_19]
  properties {bShouldExcludeFromBranchCulling=true, weights [1,1,1,1]}
 }
 {id=SoundClass_0, children [Mixer_1], properties {soundClassOverride=&Fort_Player_Foley_3P_Enemy_Athena}}
 {id=SoundClass_1, children [Mixer_0], properties {soundClassOverride=&Fort_Player_Foley_1P_Athena}}
 {id=SoundClass_2, children [Mixer_2], properties {soundClassOverride=&Fort_Player_Foley_3P_Enemy_Athena}}
 {id=Switch_0, children [Attenuation_0, Attenuation_0, Attenuation_1, Attenuation_2], properties {intParameterName=PlayerAffliation}}
 {id=WavePlayer_0, properties {soundWaveAssetPtr=&/Chrome_Build_Phase_Wall_Metal_Side_01}}
 {id=WavePlayer_1, properties {soundWaveAssetPtr=&/Chrome_Build_Phase_Wall_Metal_Side_02}}
 {id=WavePlayer_13, properties {soundWaveAssetPtr=&Chrome_Build_Phase_Wall_Metal_Side_02}}
 {id=WavePlayer_14, properties {soundWaveAssetPtr=&/Chrome_Build_Phase_Wall_Metal_Side_03}}
 {id=WavePlayer_15, properties {soundWaveAssetPtr=&/Chrome_Build_Phase_Wall_Metal_Side_04}}
 {id=WavePlayer_16, properties {soundWaveAssetPtr=&Chrome_Build_Phase_Wall_Metal_Side_01}}
 {id=WavePlayer_17, properties {soundWaveAssetPtr=&Chrome_Build_Phase_Wall_Metal_Side_02}}
 {id=WavePlayer_18, properties {soundWaveAssetPtr=&Chrome_Build_Phase_Wall_Metal_Side_03}}
 {id=WavePlayer_19, properties {soundWaveAssetPtr=&Chrome_Build_Phase_Wall_Metal_Side_04}}
 {id=WavePlayer_2, properties {soundWaveAssetPtr=&Chrome_Build_Phase_Wall_Metal_Side_03}}
 {id=WavePlayer_3, properties {soundWaveAssetPtr=&Chrome_Build_Phase_Wall_Metal_Side_04}}
 {id=WavePlayer_8, properties {soundWaveAssetPtr=&Chrome_Build_Phase_Wall_Metal_Side_01}}
]
properties {
 bOverrideConcurrency=true
 maxDistance=8000
 virtualizationMode=Disabled
 volumeMultiplier=1.2342
 concurrencyOverrides {maxCount=3, resolutionRule=StopLowestPriority}
}
