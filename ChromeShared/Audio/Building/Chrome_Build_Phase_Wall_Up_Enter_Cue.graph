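# Sound cue for entering a phase through a Chrome build wall, "Up" variant, using metal wall sounds.
# A PlayerAffliation switch routes the first two inputs to a loud first-person branch with focus attenuation, and the others to two third-person enemy-class branches, one much quieter, sharing standard wall attenuation.
# Each branch randomly picks from the same four waves with upward pitch variation, and a concurrency override caps simultaneous instances.
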SoundCue /ChromeShared/Audio/Building/Chrome_Build_Phase_Wall_Up_Enter_Cue
root=Switch_0
nodes [
 {id=Attenuation_0, children [SoundClass_1], properties {attenuationSettings=&/Chrome_Build_Phase_Wall_Focus_Att}}
 {id=Attenuation_1, children [SoundClass_0], properties {attenuationSettings=&/Chrome_Build_Phase_Wall_Att}}
 {id=Attenuation_2, children [SoundClass_2], properties {attenuationSettings=&Chrome_Build_Phase_Wall_Att}}
 {id=Mixer_0, children [Modulator_0], properties {inputVolume [4.35]}}
 {id=Mixer_1, children [Modulator_3], properties {inputVolume [0.1]}}
 {id=Mixer_2, children [Modulator_4], properties {inputVolume [1]}}
 {id=Modulator_0, children [Random_1], properties {pitchMax=1.5, pitchMin=1}}
 {id=Modulator_3, children [Random_3], properties {pitchMax=1.5, pitchMin=1}}
 {id=Modulator_4, children [Random_4], properties {pitchMax=1.5, pitchMin=1}}
 {
  id=Random_1
  children [WavePlayer_0,WavePlayer_5,WavePlayer_6,WavePlayer_7]
  properties {bShouldExcludeFromBranchCulling=true, weights [1,1,1,1]}
 }
 {
  id=Random_3
  children [WavePlayer_20, WavePlayer_21, WavePlayer_22, WavePlayer_23]
  properties {bShouldExcludeFromBranchCulling=true, weights [1,1,1,1]}
 }
 {
  id=Random_4
  children [WavePlayer_24, WavePlayer_25, WavePlayer_26, WavePlayer_27]
  properties {bShouldExcludeFromBranchCulling=true, weights [1,1,1,1]}
 }
 {id=SoundClass_0, children [Mixer_1], properties {soundClassOverride=&Fort_Player_Foley_3P_Enemy_Athena}}
 {id=SoundClass_1, children [Mixer_0], properties {soundClassOverride=&Fort_Player_Foley_1P_Athena}}
 {id=SoundClass_2, children [Mixer_2], properties {soundClassOverride=&Fort_Player_Foley_3P_Enemy_Athena}}
 {id=Switch_0, children [Attenuation_0, Attenuation_0, Attenuation_1, Attenuation_2], properties {intParameterName=PlayerAffliation}}
 {id=WavePlayer_0, properties {soundWaveAssetPtr=&/Chrome_Build_Phase_Wall_Metal_Up_04}}
 {id=WavePlayer_20, properties {soundWaveAssetPtr=&Chrome_Build_Phase_Wall_Metal_Up_04}}
 {id=WavePlayer_21, properties {soundWaveAssetPtr=&/Chrome_Build_Phase_Wall_Metal_Up_01}}
 {id=WavePlayer_22, properties {soundWaveAssetPtr=&/Chrome_Build_Phase_Wall_Metal_Up_02}}
 {id=WavePlayer_23, properties {soundWaveAssetPtr=&/Chrome_Build_Phase_Wall_Metal_Up_03}}
 {id=WavePlayer_24, properties {soundWaveAssetPtr=&Chrome_Build_Phase_Wall_Metal_Up_04}}
 {id=WavePlayer_25, properties {soundWaveAssetPtr=&Chrome_Build_Phase_Wall_Metal_Up_01}}
 {id=WavePlayer_26, properties {soundWaveAssetPtr=&Chrome_Build_Phase_Wall_Metal_Up_02}}
 {id=WavePlayer_27, properties {soundWaveAssetPtr=&Chrome_Build_Phase_Wall_Metal_Up_03}}
 {id=WavePlayer_5, properties {soundWaveAssetPtr=&Chrome_Build_Phase_Wall_Metal_Up_01}}
 {id=WavePlayer_6, properties {soundWaveAssetPtr=&Chrome_Build_Phase_Wall_Metal_Up_02}}
 {id=WavePlayer_7, properties {soundWaveAssetPtr=&Chrome_Build_Phase_Wall_Metal_Up_03}}
]
properties {
 bOverrideConcurrency=true
 maxDistance=8000
 virtualizationMode=Disabled
 volumeMultiplier=1.122
 concurrencyOverrides {maxCount=3, resolutionRule=StopLowestPriority}
}
