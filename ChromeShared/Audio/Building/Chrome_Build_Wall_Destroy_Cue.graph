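# Sound cue for a Chrome building wall being destroyed, classed under Fort impacts.
# A Random node picks among eight equally weighted destroy waves and is excluded from branch culling.
# A Modulator then randomizes pitch at fixed volume.
# The cue uses phase-wall attenuation, disables virtualization, and overrides concurrency to cap simultaneous instances by stopping the lowest priority, with priority-based volume scaling.
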SoundCue /ChromeShared/Audio/Building/Chrome_Build_Wall_Destroy_Cue
root=Modulator_2
nodes [
 {id=Modulator_2, children [Random_1], properties {pitchMax=1.2, volumeMax=1, volumeMin=1}}
 {
  id=Random_1
  children [WavePlayer_0, WavePlayer_5, WavePlayer_6, WavePlayer_7, WavePlayer_8, WavePlayer_9, WavePlayer_10, WavePlayer_11]
  properties {bShouldExcludeFromBranchCulling=true, weights [1,1,1,1,1,1,1,1]}
 }
 {id=WavePlayer_0, properties {soundWaveAssetPtr=&/Chrome_Build_Wall_Destroy_01}}
 {id=WavePlayer_10, properties {soundWaveAssetPtr=&/Chrome_Build_Wall_Destroy_07}}
 {id=WavePlayer_11, properties {soundWaveAssetPtr=&/Chrome_Build_Wall_Destroy_08}}
 {id=WavePlayer_5, properties {soundWaveAssetPtr=&/Chrome_Build_Wall_Destroy_02}}
 {id=WavePlayer_6, properties {soundWaveAssetPtr=&/Chrome_Build_Wall_Destroy_03}}
 {id=WavePlayer_7, properties {soundWaveAssetPtr=&/Chrome_Build_Wall_Destroy_04}}
 {id=WavePlayer_8, properties {soundWaveAssetPtr=&/Chrome_Build_Wall_Destroy_05}}
 {id=WavePlayer_9, properties {soundWaveAssetPtr=&/Chrome_Build_Wall_Destroy_06}}
]
properties {
 attenuationSettings=&/Chrome_Build_Phase_Wall_Att
 bOverrideConcurrency=true
 maxDistance=8000
 soundClassObject=&Fort_Impacts
 virtualizationMode=Disabled
 volumeMultiplier=1.2
 concurrencyOverrides {maxCount=3, resolutionRule=StopLowestPriority, volumeScale=0.5, volumeScaleMode=Priority}
}
