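# Sound cue named for the Chrome building wall heal effect, with an empty node graph and no root, so it references no audio to play.
# It retains playback settings: a Chrome build-phase wall attenuation asset, the Fort_Gadgets sound class and a low volume multiplier.
# A concurrency override caps simultaneous instances and prevents new ones at the limit, and virtualization is disabled.
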SoundCue /ChromeShared/Audio/Building/Chrome_Build_Wall_Heal_Cue
root=null
nodes []
properties {
 attenuationSettings=&/Chrome_Build_Phase_Wall_Att
 bOverrideConcurrency=true
 maxDistance=8000
 soundClassObject=&Fort_Gadgets
 virtualizationMode=Disabled
 volumeMultiplier=0.0446
 concurrencyOverrides {maxCount=3, resolutionRule=PreventNew}
}
