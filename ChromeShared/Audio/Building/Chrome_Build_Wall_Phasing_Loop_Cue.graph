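# Sound cue for the looping sound of Chrome build wall phasing, routed to the gadgets sound class.
# A Random node picks one of three looping wave variations, and a Modulator varies its pitch slightly.
# An Enveloper at the root ramps volume up over the first half second.
# A dedicated attenuation asset handles distance falloff, and overridden concurrency limits simultaneous voices, stopping the lowest priority with priority-based volume scaling.
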SoundCue /ChromeShared/Audio/Building/Chrome_Build_Wall_Phasing_Loop_Cue
root=Enveloper_0
nodes [
 {
  id=Enveloper_0
  children [Modulator_1]
  properties {pitchCurve {editorCurveData {keys [{time=0}]}}, volumeCurve {editorCurveData {keys [<at 0 0 RCIM_Cubic>,<at 0.5 1>]}}}
 }
 {id=Modulator_1, children [Random_1], properties {pitchMax=1.2, pitchMin=0.9, volumeMax=1, volumeMin=1}}
 {id=Random_1, children [WavePlayer_5,WavePlayer_6,WavePlayer_7], properties {bShouldExcludeFromBranchCulling=true, weights [1,1,1]}}
 {id=WavePlayer_5, properties {bLooping=true, soundWaveAssetPtr=&/Chrome_Build_Wall_Phasing_Loop_01}}
 {id=WavePlayer_6, properties {bLooping=true, soundWaveAssetPtr=&/Chrome_Build_Wall_Phasing_Loop_02}}
 {id=WavePlayer_7, properties {bLooping=true, soundWaveAssetPtr=&/Chrome_Build_Wall_Phasing_Loop_03}}
]
properties {
 attenuationSettings=&/Chrome_Build_Wall_Phasing_Loop_Att
 bOverrideConcurrency=true
 maxDistance=1000
 soundClassObject=&Fort_Gadgets
 volumeMultiplier=0.2674
 concurrencyOverrides {maxCount=3, resolutionRule=StopLowestPriority, voiceStealReleaseTime=1, volumeScale=0.7, volumeScaleMode=Priority}
}
