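# Sound cue for the proximity loop of Chrome phasable building pieces, with attenuation and a small concurrency limit.
# A PlayerAffliation switch selects a first-person branch or one of two differently scaled branches, both using the third-person enemy sound class.
# Each branch mixes a random pitch-modulated loop, faded in by an envelope, with two layered random start one-shots.
# Quality-level nodes leave the start layers empty in their last slot.
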SoundCue /ChromeShared/Audio/Building/Chrome_Phasable_Proximity_Loop_Cue
root=Switch_0
nodes [
 {
  id=Enveloper_0
  children [Modulator_0]
  properties {
   pitchCurve {editorCurveData {keys [{time=0}]}}
   volumeCurve {editorCurveData {keys [
      {arriveTangent=1.5527, interpMode=RCIM_Cubic, leaveTangent=1.5527, tangentMode=RCTM_User, time=0, value=0}
      <at 1 1>
      <at 1.23 1>
      <at 10 0.8>
     ]}}
  }
 }
 {
  id=Enveloper_1
  children [Modulator_6]
  properties {
   pitchCurve {editorCurveData {keys [{time=0}]}}
   volumeCurve {editorCurveData {keys [
      {arriveTangent=1.5527, interpMode=RCIM_Cubic, leaveTangent=1.5527, tangentMode=RCTM_User, time=0, value=0}
      <at 1 1>
      <at 1.23 1>
      <at 10 0.8>
     ]}}
  }
 }
 {
  id=Enveloper_2
  children [Modulator_9]
  properties {
   pitchCurve {editorCurveData {keys [{time=0}]}}
   volumeCurve {editorCurveData {keys [
      {arriveTangent=1.5527, interpMode=RCIM_Cubic, leaveTangent=1.5527, tangentMode=RCTM_User, time=0, value=0}
      <at 1 1>
      <at 1.23 1>
      <at 10 0.8>
     ]}}
  }
 }
 {id=Mixer_0, children [Enveloper_0, QualityLevel_0, null], properties {inputVolume [1.5,3,0.5]}}
 {id=Mixer_1, children [Modulator_1,Modulator_3], properties {inputVolume [1,1]}}
 {id=Mixer_2, children [Mixer_0], properties {inputVolume [2.5]}}
 {id=Mixer_3, children [Mixer_5], properties {inputVolume [0.1]}}
 {id=Mixer_4, children [Mixer_7], properties {inputVolume [1.8]}}
 {id=Mixer_5, children [Enveloper_1, QualityLevel_1, null], properties {inputVolume [1.5,3,0.5]}}
 {id=Mixer_6, children [Modulator_7,Modulator_8], properties {inputVolume [1,1]}}
 {id=Mixer_7, children [Enveloper_2, QualityLevel_2, null], properties {inputVolume [1.5,3,0.5]}}
 {id=Mixer_8, children [Modulator_10,Modulator_11], properties {inputVolume [1,1]}}
 {id=Modulator_0, children [Random_0], properties {pitchMax=1.2, pitchMin=0.9, volumeMax=1, volumeMin=1}}
 {id=Modulator_1, children [Random_2], properties {pitchMax=1, pitchMin=0.8, volumeMax=1, volumeMin=1}}
 {id=Modulator_10, children [Random_10], properties {pitchMax=1, pitchMin=0.8, volumeMax=1, volumeMin=1}}
 {id=Modulator_11, children [Random_11], properties {pitchMax=1, pitchMin=0.8, volumeMax=1, volumeMin=0.6}}
 {id=Modulator_3, children [Random_3], properties {pitchMax=1, pitchMin=0.8, volumeMax=1, volumeMin=0.6}}
 {id=Modulator_6, children [Random_6], properties {pitchMax=1.2, pitchMin=0.9, volumeMax=1, volumeMin=1}}
 {id=Modulator_7, children [Random_7], properties {pitchMax=1, pitchMin=0.8, volumeMax=1, volumeMin=1}}
 {id=Modulator_8, children [Random_8], properties {pitchMax=1, pitchMin=0.8, volumeMax=1, volumeMin=0.6}}
 {id=Modulator_9, children [Random_9], properties {pitchMax=1.2, pitchMin=0.9, volumeMax=1, volumeMin=1}}
 {id=QualityLevel_0, children [Mixer_1,Mixer_1,Mixer_1,null]}
 {id=QualityLevel_1, children [Mixer_6,Mixer_6,Mixer_6,null]}
 {id=QualityLevel_2, children [Mixer_8,Mixer_8,Mixer_8,null]}
 {
  id=Random_0
  children [WavePlayer_1,WavePlayer_2,WavePlayer_3,WavePlayer_4]
  properties {bShouldExcludeFromBranchCulling=true, weights [1,1,1,1]}
 }
 {
  id=Random_10
  children [WavePlayer_44, WavePlayer_45, WavePlayer_46, WavePlayer_47, WavePlayer_48]
  properties {bShouldExcludeFromBranchCulling=true, weights [1,1,1,1,1]}
 }
 {
  id=Random_11
  children [WavePlayer_49, WavePlayer_50, WavePlayer_51, WavePlayer_52, WavePlayer_53]
  properties {bShouldExcludeFromBranchCulling=true, weights [1,1,1,1,1]}
 }
 {
  id=Random_2
  children [WavePlayer_12, WavePlayer_13, WavePlayer_14, WavePlayer_15, WavePlayer_16]
  properties {bShouldExcludeFromBranchCulling=true, weights [1,1,1,1,1]}
 }
 {
  id=Random_3
  children [WavePlayer_17, WavePlayer_18, WavePlayer_19, WavePlayer_20, WavePlayer_21]
  properties {bShouldExcludeFromBranchCulling=true, weights [1,1,1,1,1]}
 }
 {
  id=Random_6
  children [WavePlayer_10, WavePlayer_11, WavePlayer_28, WavePlayer_29]
  properties {bShouldExcludeFromBranchCulling=true, weights [1,1,1,1]}
 }
 {
  id=Random_7
  children [WavePlayer_30, WavePlayer_31, WavePlayer_32, WavePlayer_33, WavePlayer_34]
  properties {bShouldExcludeFromBranchCulling=true, weights [1,1,1,1,1]}
 }
 {
  id=Random_8
  children [WavePlayer_35, WavePlayer_36, WavePlayer_37, WavePlayer_38, WavePlayer_39]
  properties {bShouldExcludeFromBranchCulling=true, weights [1,1,1,1,1]}
 }
 {
  id=Random_9
  children [WavePlayer_40, WavePlayer_41, WavePlayer_42, WavePlayer_43]
  properties {bShouldExcludeFromBranchCulling=true, weights [1,1,1,1]}
 }
 {id=SoundClass_0, children [Mixer_3], properties {soundClassOverride=&Fort_Player_Foley_3P_Enemy_Athena}}
 {id=SoundClass_1, children [Mixer_2], properties {soundClassOverride=&Fort_Player_Foley_1P_Athena}}
 {id=SoundClass_2, children [Mixer_4], properties {soundClassOverride=&Fort_Player_Foley_3P_Enemy_Athena}}
 {id=Switch_0, children [SoundClass_1,SoundClass_1,SoundClass_0,SoundClass_2], properties {intParameterName=PlayerAffliation}}
 {id=WavePlayer_1, properties {bLooping=true, soundWaveAssetPtr=&/Chrome_Phasable_Proximity_Loop_01}}
 {id=WavePlayer_10, properties {bLooping=true, soundWaveAssetPtr=&Chrome_Phasable_Proximity_Loop_01}}
 {id=WavePlayer_11, properties {bLooping=true, soundWaveAssetPtr=&/Chrome_Phasable_Proximity_Loop_02}}
 {id=WavePlayer_12, properties {soundWaveAssetPtr=&/Chrome_Phasable_Proximity_Start_A_01}}
 {id=WavePlayer_13, properties {soundWaveAssetPtr=&/Chrome_Phasable_Proximity_Start_A_02}}
 {id=WavePlayer_14, properties {soundWaveAssetPtr=&/Chrome_Phasable_Proximity_Start_A_03}}
 {id=WavePlayer_15, properties {soundWaveAssetPtr=&/Chrome_Phasable_Proximity_Start_A_04}}
 {id=WavePlayer_16, properties {soundWaveAssetPtr=&/Chrome_Phasable_Proximity_Start_A_05}}
 {id=WavePlayer_17, properties {soundWaveAssetPtr=&/Chrome_Phasable_Proximity_Start_B_01}}
 {id=WavePlayer_18, properties {soundWaveAssetPtr=&/Chrome_Phasable_Proximity_Start_B_02}}
 {id=WavePlayer_19, properties {soundWaveAssetPtr=&/Chrome_Phasable_Proximity_Start_B_03}}
 {id=WavePlayer_2, properties {bLooping=true, soundWaveAssetPtr=&Chrome_Phasable_Proximity_Loop_02}}
 {id=WavePlayer_20, properties {soundWaveAssetPtr=&/Chrome_Phasable_Proximity_Start_B_04}}
 {id=WavePlayer_21, properties {soundWaveAssetPtr=&/Chrome_Phasable_Proximity_Start_B_05}}
 {id=WavePlayer_28, properties {bLooping=true, soundWaveAssetPtr=&/Chrome_Phasable_Proximity_Loop_03}}
 {id=WavePlayer_29, properties {bLooping=true, soundWaveAssetPtr=&/Chrome_Phasable_Proximity_Loop_04}}
 {id=WavePlayer_3, properties {bLooping=true, soundWaveAssetPtr=&Chrome_Phasable_Proximity_Loop_03}}
 {id=WavePlayer_30, properties {soundWaveAssetPtr=&Chrome_Phasable_Proximity_Start_A_01}}
 {id=WavePlayer_31, properties {soundWaveAssetPtr=&Chrome_Phasable_Proximity_Start_A_02}}
 {id=WavePlayer_32, properties {soundWaveAssetPtr=&Chrome_Phasable_Proximity_Start_A_03}}
 {id=WavePlayer_33, properties {soundWaveAssetPtr=&Chrome_Phasable_Proximity_Start_A_04}}
 {id=WavePlayer_34, properties {soundWaveAssetPtr=&Chrome_Phasable_Proximity_Start_A_05}}
 {id=WavePlayer_35, properties {soundWaveAssetPtr=&Chrome_Phasable_Proximity_Start_B_01}}
 {id=WavePlayer_36, properties {soundWaveAssetPtr=&Chrome_Phasable_Proximity_Start_B_02}}
 {id=WavePlayer_37, properties {soundWaveAssetPtr=&Chrome_Phasable_Proximity_Start_B_03}}
 {id=WavePlayer_38, properties {soundWaveAssetPtr=&Chrome_Phasable_Proximity_Start_B_04}}
 {id=WavePlayer_39, properties {soundWaveAssetPtr=&Chrome_Phasable_Proximity_Start_B_05}}
 {id=WavePlayer_4, properties {bLooping=true, soundWaveAssetPtr=&Chrome_Phasable_Proximity_Loop_04}}
 {id=WavePlayer_40, properties {bLooping=true, soundWaveAssetPtr=&Chrome_Phasable_Proximity_Loop_01}}
 {id=WavePlayer_41, properties {bLooping=true, soundWaveAssetPtr=&Chrome_Phasable_Proximity_Loop_02}}
 {id=WavePlayer_42, properties {bLooping=true, soundWaveAssetPtr=&Chrome_Phasable_Proximity_Loop_03}}
 {id=WavePlayer_43, properties {bLooping=true, soundWaveAssetPtr=&Chrome_Phasable_Proximity_Loop_04}}
 {id=WavePlayer_44, properties {soundWaveAssetPtr=&Chrome_Phasable_Proximity_Start_A_01}}
 {id=WavePlayer_45, properties {soundWaveAssetPtr=&Chrome_Phasable_Proximity_Start_A_02}}
 {id=WavePlayer_46, properties {soundWaveAssetPtr=&Chrome_Phasable_Proximity_Start_A_03}}
 {id=WavePlayer_47, properties {soundWaveAssetPtr=&Chrome_Phasable_Proximity_Start_A_04}}
 {id=WavePlayer_48, properties {soundWaveAssetPtr=&Chrome_Phasable_Proximity_Start_A_05}}
 {id=WavePlayer_49, properties {soundWaveAssetPtr=&Chrome_Phasable_Proximity_Start_B_01}}
 {id=WavePlayer_50, properties {soundWaveAssetPtr=&Chrome_Phasable_Proximity_Start_B_02}}
 {id=WavePlayer_51, properties {soundWaveAssetPtr=&Chrome_Phasable_Proximity_Start_B_03}}
 {id=WavePlayer_52, properties {soundWaveAssetPtr=&Chrome_Phasable_Proximity_Start_B_04}}
 {id=WavePlayer_53, properties {soundWaveAssetPtr=&Chrome_Phasable_Proximity_Start_B_05}}
]
properties {
 attenuationSettings=&/Chrome_Build_Wall_Proximity_Loop_Att
 bOverrideConcurrency=true
 maxDistance=1850
 volumeMultiplier=0.8913
 concurrencyOverrides {maxCount=3, volumeScale=0.7, volumeScaleMode=Priority}
}
