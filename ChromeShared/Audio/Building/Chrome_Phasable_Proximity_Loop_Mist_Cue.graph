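# Looping proximity sound cue for phasable Chrome building pieces, playing sparkle loops within short-range attenuation under a three-voice concurrency cap.
# An Enveloper fades volume in and shapes pitch over the start, feeding a quality node whose populated levels share one switch.
# The PlayerAffliation switch selects a first-person or one of two enemy third-person sound-class branches at different mix levels.
# Each branch randomly picks between two loops with pitch variation.
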SoundCue /ChromeShared/Audio/Building/Chrome_Phasable_Proximity_Loop_Mist_Cue
root=Enveloper_3
nodes [
 {
  id=Enveloper_3
  children [QualityLevel_6]
  properties {pitchCurve {editorCurveData {keys [<at 0 1.25 RCIM_Cubic>, {time=1.5}]}}, volumeCurve {editorCurveData {keys [<at 0 0>,<at 0.5 1>]}}}
 }
 {id=Mixer_2, children [Modulator_12], properties {inputVolume [2.5]}}
 {id=Mixer_3, children [Modulator_13], properties {inputVolume [0.1]}}
 {id=Mixer_4, children [Modulator_14], properties {inputVolume [1.8]}}
 {id=Modulator_12, children [Random_12], properties {pitchMax=1.2, pitchMin=0.9, volumeMax=1, volumeMin=1}}
 {id=Modulator_13, children [Random_13], properties {pitchMax=1.2, pitchMin=0.9, volumeMax=1, volumeMin=1}}
 {id=Modulator_14, children [Random_14], properties {pitchMax=1.2, pitchMin=0.9, volumeMax=1, volumeMin=1}}
 {id=QualityLevel_6, children [Switch_0,Switch_0,Switch_0,null]}
 {id=Random_12, children [WavePlayer_105, WavePlayer_106], properties {weights [1,1]}}
 {id=Random_13, children [WavePlayer_107, WavePlayer_108], properties {weights [1,1]}}
 {id=Random_14, children [WavePlayer_109, WavePlayer_110], properties {weights [1,1]}}
 {id=SoundClass_0, children [Mixer_3], properties {soundClassOverride=&Fort_Player_Foley_3P_Enemy_Athena}}
 {id=SoundClass_1, children [Mixer_2], properties {soundClassOverride=&Fort_Player_Foley_1P_Athena}}
 {id=SoundClass_2, children [Mixer_4], properties {soundClassOverride=&Fort_Player_Foley_3P_Enemy_Athena}}
 {id=Switch_0, children [SoundClass_1,SoundClass_1,SoundClass_0,SoundClass_2], properties {intParameterName=PlayerAffliation}}
 {id=WavePlayer_105, properties {bLooping=true, soundWaveAssetPtr=&/Chrome_Phasable_Proximity_Loop_Sparkles_01}}
 {id=WavePlayer_106, properties {bLooping=true, soundWaveAssetPtr=&/Chrome_Phasable_Proximity_Loop_Sparkles_02}}
 {id=WavePlayer_107, properties {bLooping=true, soundWaveAssetPtr=&Chrome_Phasable_Proximity_Loop_Sparkles_01}}
 {id=WavePlayer_108, properties {bLooping=true, soundWaveAssetPtr=&Chrome_Phasable_Proximity_Loop_Sparkles_02}}
 {id=WavePlayer_109, properties {bLooping=true, soundWaveAssetPtr=&Chrome_Phasable_Proximity_Loop_Sparkles_01}}
 {id=WavePlayer_110, properties {bLooping=true, soundWaveAssetPtr=&Chrome_Phasable_Proximity_Loop_Sparkles_02}}
]
properties {
 attenuationSettings=&/Chrome_Build_Wall_Proximity_Loop_Mist_Att
 bOverrideConcurrency=true
 maxDistance=650
 volumeMultiplier=2
 concurrencyOverrides {maxCount=3, volumeScale=0.7, volumeScaleMode=Priority}
}
